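# Sound cue that plays GreenTown PA speaker voice-over lines on an endless loop.
# Each loop waits a random interval of several seconds, plus a short fixed delay, then picks one of twenty equally weighted voice clips.
# Base submix and submix sends are disabled, with audio sent pre-effect to the SB_GreenTown_PA_Speaker source bus, so the cue feeds that bus rather than the main mix.
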SoundCue /Game/Sounds/Helios/POI/GreenTown/PA_Speakers/SC_VO_GreenTown_PA_Speaker
root=Looping_0
nodes [
 {id=Delay_0, children [Random_0], properties {delayMax=1, delayMin=1}}
 {id=Delay_1, children [Delay_0], properties {delayMax=30, delayMin=15}}
 {id=Looping_0, children [Delay_1]}
 {
  id=Random_0
  children [
   WavePlayer_0
   WavePlayer_1
   WavePlayer_2
   WavePlayer_3
   WavePlayer_4
   WavePlayer_5
   WavePlayer_6
   WavePlayer_7
   WavePlayer_8
   WavePlayer_9
   WavePlayer_10
   WavePlayer_11
   WavePlayer_12
   WavePlayer_13
   WavePlayer_14
   WavePlayer_15
   WavePlayer_16
   WavePlayer_17
   WavePlayer_18
   WavePlayer_19
  ]
  properties {bShouldExcludeFromBranchCulling=true, weights [1,1,1,1,1,1,1,1,1,1,1,1,1,1,1,1,1,1,1,1]}
 }
 {id=WavePlayer_0, properties {soundWaveAssetPtr=&/VO_GreenTown_PA_Speaker_04}}
 {id=WavePlayer_1, properties {soundWaveAssetPtr=&/VO_GreenTown_PA_Speaker_01}}
 {id=WavePlayer_10, properties {soundWaveAssetPtr=&/VO_GreenTown_PA_Speaker_18}}
 {id=WavePlayer_11, properties {soundWaveAssetPtr=&/VO_GreenTown_PA_Speaker_19}}
 {id=WavePlayer_12, properties {soundWaveAssetPtr=&/VO_GreenTown_PA_Speaker_20}}
 {id=WavePlayer_13, properties {soundWaveAssetPtr=&/VO_GreenTown_PA_Speaker_21}}
 {id=WavePlayer_14, properties {soundWaveAssetPtr=&/VO_GreenTown_PA_Speaker_22}}
 {id=WavePlayer_15, properties {soundWaveAssetPtr=&/VO_GreenTown_PA_Speaker_23}}
 {id=WavePlayer_16, properties {soundWaveAssetPtr=&/VO_GreenTown_PA_Speaker_27}}
 {id=WavePlayer_17, properties {soundWaveAssetPtr=&/VO_GreenTown_PA_Speaker_28}}
 {id=WavePlayer_18, properties {soundWaveAssetPtr=&/VO_GreenTown_PA_Speaker_29}}
 {id=WavePlayer_19, properties {soundWaveAssetPtr=&/VO_GreenTown_PA_Speaker_30}}
 {id=WavePlayer_2, properties {soundWaveAssetPtr=&/VO_GreenTown_PA_Speaker_10}}
 {id=WavePlayer_3, properties {soundWaveAssetPtr=&/VO_GreenTown_PA_Speaker_09}}
 {id=WavePlayer_4, properties {soundWaveAssetPtr=&/VO_GreenTown_PA_Speaker_08}}
 {id=WavePlayer_5, properties {soundWaveAssetPtr=&/VO_GreenTown_PA_Speaker_07}}
 {id=WavePlayer_6, properties {soundWaveAssetPtr=&/VO_GreenTown_PA_Speaker_06}}
 {id=WavePlayer_7, properties {soundWaveAssetPtr=&/VO_GreenTown_PA_Speaker_05}}
 {id=WavePlayer_8, properties {soundWaveAssetPtr=&/VO_GreenTown_PA_Speaker_16}}
 {id=WavePlayer_9, properties {soundWaveAssetPtr=&/VO_GreenTown_PA_Speaker_17}}
]
properties {
 bEnableBaseSubmix=false
 bEnableSubmixSends=false
 maxDistance=2097152
 volumeMultiplier=0.6
 concurrencyOverrides {maxCount=3}
 preEffectBusSends [{soundSourceBus=&/SB_GreenTown_PA_Speaker}]
}
